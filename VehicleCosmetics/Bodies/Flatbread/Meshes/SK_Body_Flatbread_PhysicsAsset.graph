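# Physics asset for the skeletal mesh of the Flatbread vehicle body cosmetic.
# Each bone body is a single capsule, covering frame, body shell, glass, front differential, steering, wheels, brake pads and rear shock bases.
# Constraints form a tree rooted at the frame: glass hangs off the body, front wheels chain through differential and knuckles, rear parts attach to axle pivots.
# All joints share uniform 45-degree swing and twist limit values, resembling auto-generated defaults rather than tuned suspension.
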PhysicsAsset /VehicleCosmetics/Bodies/Flatbread/Meshes/SK_Body_Flatbread_PhysicsAsset
bodies {
 axle_pivot_bk_l [{shapeType=Capsule, length=34.5234, radius=26.7737, center=<-2.5232 -16.9758 -4.82>, rotation=<rot 0 -81.4686 -87.5277>}]
 axle_pivot_bk_r [{shapeType=Capsule, length=34.5234, radius=26.7737, center=<-2.5232 16.9758 -4.82>, rotation=<rot 0 -98.5314 -87.5277>}]
 body [{shapeType=Capsule, length=271.5572, radius=135.7947, center=<22.1018 -0.0484 11.3077>, rotation=<rot 0 -89.9484 -83.6621>}]
 brakepad_bk_l [{shapeType=Capsule, length=21.7855, radius=13.5573, center=<-5.6561 2.2811 5.8148>, rotation=<rot -13.1403 -90 -41.7153>}]
 brakepad_bk_r [{shapeType=Capsule, length=21.7428, radius=13.5548, center=<-5.6291 -2.4233 5.8115>, rotation=<rot 12.3911 -90 -41.7984>}]
 frame [{shapeType=Capsule, length=0.505, radius=0.505, rotation=<rot 0 -90 0>}]
 front_differential [{shapeType=Capsule, length=69.4798, radius=6.9951, rotation=<rot 0 0 -90>}]
 glass_c [{shapeType=Capsule, length=99.4986, radius=46.5798, center=<-73.9679 0.7202 16.8784>, rotation=<rot 0 -89.2933 -73.9152>}]
 glass_l [{shapeType=Capsule, length=71.1844, radius=7.5588, center=<-47.9759 13.6079 12.4035>, rotation=<rot 0 -102.2168 -82.4665>}]
 glass_r [{shapeType=Capsule, length=71.1844, radius=7.5588, center=<47.9759 -13.6079 -12.4035>, rotation=<rot 0 -102.2168 -82.4665>}]
 shock_base_bk_l [{shapeType=Capsule, length=13.6841, radius=13.6264, center=<0 -0.8934 -0.0171>, rotation=<rot -0.8394 -90 -13.3293>}]
 shock_base_bk_r [{shapeType=Capsule, length=13.6982, radius=13.6457, center=<0 0.8936 -0.016>, rotation=<rot 0.7873 -90 -14.4459>}]
 steering_knuckle_fr_l [{shapeType=Capsule, length=27.6226, radius=19.5867, center=<-3.5388 -32.7976 0.3962>, rotation=<rot 0 174.9578 -84.899>}]
 steering_knuckle_fr_r [{shapeType=Capsule, length=27.6407, radius=19.5575, center=<-3.5546 32.7838 0.4091>, rotation=<rot 0 5.1526 -84.8286>}]
 wheel_disc_bk_l [{shapeType=Capsule, length=25.9714, radius=25.9684, center=<0 -2.7364 0.0032>, rotation=<rot 0.0838 -90 -16.3852>}]
 wheel_disc_bk_r [{shapeType=Capsule, length=25.9714, radius=25.9684, center=<0 2.7364 0.0032>, rotation=<rot -0.0838 -90 -16.3852>}]
 wheel_disc_fr_l [{shapeType=Capsule, length=22.8012, radius=22.7986, center=<0 -1.996 0.0028>, rotation=<rot 0.0838 -90 -16.3853>}]
 wheel_disc_fr_r [{shapeType=Capsule, length=22.8012, radius=22.7986, center=<0 1.996 0.0028>, rotation=<rot -0.0838 -90 -16.3851>}]
 wheel_steering_fr_l [{shapeType=Capsule, length=19.1055, radius=11.9325, center=<-4.8766 -3.6978 5.1821>, rotation=<rot -12.7945 -90 -40.9297>}]
 wheel_steering_fr_r [{shapeType=Capsule, length=19.1003, radius=11.9124, center=<-4.8574 3.7677 5.1776>, rotation=<rot 13.0707 -90 -40.8991>}]
}
constraints [
 {bone1Name=body, bone2Name=frame, swing1LimitDegrees=45, swing2LimitDegrees=45, twistLimitDegrees=45}
 {bone1Name=glass_c, bone2Name=body, swing1LimitDegrees=45, swing2LimitDegrees=45, twistLimitDegrees=45}
 {bone1Name=glass_l, bone2Name=body, swing1LimitDegrees=45, swing2LimitDegrees=45, twistLimitDegrees=45}
 {bone1Name=glass_r, bone2Name=body, swing1LimitDegrees=45, swing2LimitDegrees=45, twistLimitDegrees=45}
 {bone1Name=front_differential, bone2Name=frame, swing1LimitDegrees=45, swing2LimitDegrees=45, twistLimitDegrees=45}
 {bone1Name=steering_knuckle_fr_l, bone2Name=front_differential, swing1LimitDegrees=45, swing2LimitDegrees=45, twistLimitDegrees=45}
 {bone1Name=wheel_steering_fr_l, bone2Name=steering_knuckle_fr_l, swing1LimitDegrees=45, swing2LimitDegrees=45, twistLimitDegrees=45}
 {bone1Name=wheel_disc_fr_l, bone2Name=wheel_steering_fr_l, swing1LimitDegrees=45, swing2LimitDegrees=45, twistLimitDegrees=45}
 {bone1Name=steering_knuckle_fr_r, bone2Name=front_differential, swing1LimitDegrees=45, swing2LimitDegrees=45, twistLimitDegrees=45}
 {bone1Name=wheel_steering_fr_r, bone2Name=steering_knuckle_fr_r, swing1LimitDegrees=45, swing2LimitDegrees=45, twistLimitDegrees=45}
 {bone1Name=wheel_disc_fr_r, bone2Name=wheel_steering_fr_r, swing1LimitDegrees=45, swing2LimitDegrees=45, twistLimitDegrees=45}
 {bone1Name=axle_pivot_bk_l, bone2Name=frame, swing1LimitDegrees=45, swing2LimitDegrees=45, twistLimitDegrees=45}
 {bone1Name=wheel_disc_bk_l, bone2Name=axle_pivot_bk_l, swing1LimitDegrees=45, swing2LimitDegrees=45, twistLimitDegrees=45}
 {bone1Name=brakepad_bk_l, bone2Name=axle_pivot_bk_l, swing1LimitDegrees=45, swing2LimitDegrees=45, twistLimitDegrees=45}
 {bone1Name=shock_base_bk_l, bone2Name=axle_pivot_bk_l, swing1LimitDegrees=45, swing2LimitDegrees=45, twistLimitDegrees=45}
 {bone1Name=axle_pivot_bk_r, bone2Name=frame, swing1LimitDegrees=45, swing2LimitDegrees=45, twistLimitDegrees=45}
 {bone1Name=wheel_disc_bk_r, bone2Name=axle_pivot_bk_r, swing1LimitDegrees=45, swing2LimitDegrees=45, twistLimitDegrees=45}
 {bone1Name=brakepad_bk_r, bone2Name=axle_pivot_bk_r, swing1LimitDegrees=45, swing2LimitDegrees=45, twistLimitDegrees=45}
 {bone1Name=shock_base_bk_r, bone2Name=axle_pivot_bk_r, swing1LimitDegrees=45, swing2LimitDegrees=45, twistLimitDegrees=45}
]
